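# Sound cue for the impact of the Pretzel item landing in Athena.
# It plays a single landing wave through a Modulator that randomly raises the pitch on each play for variation.
# The cue is boosted in volume, uses weapon-drop attenuation and the player foley sound class, and overrides concurrency to block new instances once a small cap is reached.
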SoundCue /Game/Athena/Sounds/Items/Pretzel/Athena_Pretzel_Impact_Cue
root=Modulator_5
nodes [
 {id=Modulator_5, children [WavePlayer_0], properties {pitchMax=1.5, pitchMin=1.1}}
 {id=WavePlayer_0, properties {soundWaveAssetPtr=/Game/Athena/Sounds/Items/Pretzel/wav/Pretzel_Land}}
]
properties {
 attenuationSettings=&Fort_Weapon_Drop_Att
 bOverrideConcurrency=true
 maxDistance=3000
 soundClassObject=&Fort_Player_Foley
 volumeMultiplier=1.5
 concurrencyOverrides {maxCount=3, resolutionRule=PreventNew}
}
